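# Material function that layers parameter-driven damage onto the Doggus character body's incoming material attributes.
# IsTorsoDamage selects a mask from torso and jet textures gated by UV and vertex-color sphere masks, or from a channel-weighted FXMaskTexture.
# The mask darkens base color and specular, drives a noise height-lerped layer of exposed metal with its own roughness, and blends in a damage normal.
# ES2 and ES3.1 pass the input through unchanged.
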MaterialFunction /Figment_S10/Characters/Player/Male/Medium/Bodies/M_MED_Doggus/MaterialSet/MF_DoggusDam
graph {nodes [
  Constant() {r=1.5}
  Constant() {r=-50}
  Constant() {r=-0.5}
  Constant() {r=3.5}
  Constant() {r=2}
  Constant() {r=-10}
  Constant() {r=-0.25}
  Constant() {r=0.4}
  Constant() {r=1}
  Constant() {r=10}
  Constant() {r=10}
  n1 = TextureSample() {
   samplerSource=SSM_Wrap_WorldGroupSettings
   samplerType=SAMPLERTYPE_Masks
   texture=/Game/Characters/Player/Male/Medium/Bodies/M_MED_Doggus/Textures/T_Doggus_Torso_FX
  }
  n2 = Constant() {r=0.5}
  n3 = MaterialFunctionCall(Float2=TextureCoordinate() {uTiling=0.75, vTiling=0.75}) {materialFunction=&BreakOutFloat2Components}
  n4 = SphereMask(A=@n2, B=@n3.R) {attenuationRadius=0.5}
  n5 = SphereMask(A=@n3.G, B=@n2) {attenuationRadius=0.5}
  n6 = Multiply(A=ScalarParameter() {group=Damage, parameterName=TorsoDamage0}, B=LinearInterpolate(A=@n1.G, B=@n1.G, Alpha=Multiply(A=@n4, B=@n5)))
  n7 = Multiply(A=ScalarParameter() {group=Damage, parameterName=TorsoArmRemoveDamage}, B=@n1.B)
  n8 = ScalarParameter() {group=Damage, parameterName=TorsoDamage1}
  n9 = TextureSample() {
   samplerSource=SSM_Wrap_WorldGroupSettings
   samplerType=SAMPLERTYPE_Masks
   texture=/Game/Characters/Player/Male/Medium/Bodies/M_MED_Doggus/Textures/T_Doggus_TorsoJets_FX
  }
  n10 = ScalarParameter() {group=Damage, parameterName=TorsoDamage4}
  n11 = Saturate(None=Add(A=Multiply(A=@n8, B=@n9.R), B=Multiply(A=@n9.B, B=@n10)))
  n12 = VertexColor()
  n13 = Constant() {r=0.025}
  n14 = Constant() {r=1}
  n15 = SphereMask(A=@n12.R, B=Constant() {r=0.7}, Radius=@n13, Hardness=@n14) {attenuationRadius=0.5}
  n16 = Multiply(A=ScalarParameter() {group=Damage, parameterName=TorsoDamage2}, B=@n9.G)
  n17 = SphereMask(A=@n12.R, B=Constant() {r=0.6}, Radius=@n13, Hardness=@n14) {attenuationRadius=0.5}
  n18 = Constant() {r=0.9}
  n19 = Max(
   A=Max(A=Multiply(A=@n11, B=@n15), B=Multiply(A=@n16, B=@n17))
   B=Multiply(
    A=ScalarParameter() {desc=0-.8 only, group=Damage, parameterName=TorsoDamage3}
    B=SphereMask(A=@n12.R, B=@n18, Radius=@n13, Hardness=@n14) {attenuationRadius=0.5}
   )
  )
  n20 = TextureSampleParameter2D() {
   group=Damage
   parameterName=FXMaskTexture
   samplerSource=SSM_Wrap_WorldGroupSettings
   samplerType=SAMPLERTYPE_Masks
   texture=/Game/Characters/Player/Male/Medium/Bodies/M_MED_Doggus/Textures/T_Doggus_Arms_FX
  }
  n21 = Multiply(A=ScalarParameter() {group=Damage, parameterName=FXMaskRed}, B=@n20.R)
  n22 = Multiply(A=ScalarParameter() {group=Damage, parameterName=FXMaskGreen}, B=@n20.G)
  n23 = ScalarParameter() {group=Damage, parameterName=FXMaskBlue}
  n24 = StaticSwitchParameter(True=Max(A=Max(A=@n6, B=@n7), B=@n19), False=Max(A=Max(A=@n21, B=@n22), B=Multiply(A=@n23, B=@n20.B))) {group=Damage, parameterName=IsTorsoDamage}
  n25 = OneMinus(None=@n24)
  n26 = Add(A=@n25, B=ScalarParameter() {defaultValue=1, group=Damage, parameterName=Damage-SpecMin})
  n27 = GetMaterialAttributes(None=FunctionInput() {inputType=FunctionInput_MaterialAttributes}) {attributeGetTypes [
    #1
    #2
    #3
    #4
    #5
    #6
   ]}
  n28 = Add(
   A=ScalarParameter() {group=Damage, parameterName=DiffuseDamageBurnDarkness}
   B=Power(Base=@n25, Exp=ScalarParameter() {defaultValue=1, desc=turn this to 2 in torso MI, group=Damage, parameterName=DiffuseDamageDarkPow})
  )
  n29 = Saturate(None=@n28)
  n30 = TextureSample(UVs=Multiply(A=TextureCoordinate(), B=ScalarParameter() {defaultValue=10, group=Damage, parameterName=Micro2-RoughnessTile})) {
   samplerSource=SSM_Wrap_WorldGroupSettings
   samplerType=SAMPLERTYPE_Masks
   texture=/Game/Athena/Prototype/Blueprints/Cube/Materials/T_CubeNoise
  }
  n31 = Multiply(
   A=LinearInterpolate(
    A=ScalarParameter() {defaultValue=-0.2, group=Damage, parameterName=DiffuseDamageMin}
    B=ScalarParameter() {defaultValue=1, group=Damage, parameterName=DiffuseDamageMax}
    Alpha=@n30.R
   )
   B=ScalarParameter() {defaultValue=1, group=Damage, parameterName=ik88}
  )
  n32 = Constant() {r=-10}
  n33 = Constant() {r=1.5}
  n34 = Add(A=Saturate(None=LinearInterpolate(A=@n32, B=@n33, Alpha=@n24)), B=Add(A=Constant() {r=-0.25}, B=@n30.R) {constB=-0.5}) {constB=-1}
  n35 = Multiply(
   A=Saturate(None=MaterialFunctionCall(Transition Phase=@n34, Height Texture=@n30.R) {materialFunction=&HeightLerp}.Lerp Alpha No Contrast)
   B=@n24
  )
  n36 = Add(
   A=ScalarParameter() {defaultValue=0.4, group=Damage, parameterName=Damage-NormalOffset}
   B=Multiply(A=TextureCoordinate(), B=ScalarParameter() {defaultValue=3.5, group=Damage, parameterName=Damage-NormalTile})
  )
  n37 = LinearInterpolate(
   A=Multiply(
    A=TextureSample(UVs=@n36) {samplerType=SAMPLERTYPE_Normal, texture=/Figment_S10/Vehicles/FORT_Vehicles_01/CarDamageNormal}.RGB
    B=VectorParameter() {group=Damage, parameterName=Damage-NormalAmount, defaultValue=<rgba 2 2 1 1>}.RGB
   )
   B=Constant3Vector() {constant=<rgba 0 0 1 0>}
   Alpha=@n25
  )
  n38 = LinearInterpolate(
   A=ScalarParameter() {defaultValue=0.1, group=Damage, parameterName=Damage-zRoughA}
   B=ScalarParameter() {defaultValue=0.5, group=Damage, parameterName=Damage-zRoughB}
   Alpha=@n30.R
  )
  n39 = LinearInterpolate(
   A=ScalarParameter() {defaultValue=0.2, group=Damage, parameterName=Micro2-RoughLow}
   B=ScalarParameter() {defaultValue=-0.1, group=Damage, parameterName=Micro2-RoughHigh}
   Alpha=@n30.R
  )
  n40 = Saturate(None=Add(A=@n27.Roughness, B=Saturate(None=@n39)))
  n41 = SetMaterialAttributes!(
   MaterialAttributes=@n27.MaterialAttributes
   BaseColor=LinearInterpolate(A=Multiply(A=@n29, B=@n27.BaseColor), B=Saturate(None=@n31), Alpha=@n35)
   Metallic=Saturate(None=Add(A=@n35, B=@n27.Metallic))
   Specular=Multiply(A=Saturate(None=@n26), B=@n27.Specular)
   Roughness=LinearInterpolate(A=@n40, B=@n38, Alpha=@n35)
   EmissiveColor=@n27.EmissiveColor
   Normal=MaterialFunctionCall(BaseNormal=@n27.Normal, AdditionalNormal=Normalize(VectorInput=@n37)) {materialFunction=&BlendAngleCorrectedNormals}.Result
  ) {attributeSetTypes [
    #1
    #2
    #3
    #4
    #5
    #6
   ]}
  FunctionOutput(None=FeatureLevelSwitch(Default=@n41, ES2=@n27.MaterialAttributes, ES3_1=@n27.MaterialAttributes, SM4_REMOVED=@n41))
  LinearInterpolate(A=@n38, B=@n40, Alpha=@n25)
  n42 = Multiply(A=VertexInterpolator(VS=PreSkinnedPosition() {desc=could replace if uvs are fixed}).PS, B=Constant() {r=100})
  LinearInterpolate(Alpha=Saturate(None=ComponentMask(None=@n42) {r=true}))
  Saturate(None=Add(A=ScalarParameter() {group=Damage, parameterName=DEBUGDAM}, B=@n24))
  n43 = ScalarParameter() {group=Damage, parameterName=DiffuseDamageBurnDarknessJets}
  StaticSwitchParameter(True=LinearInterpolate(A=@n28, B=Add(A=@n43, B=@n25), Alpha=@n19), False=@n28) {group=Damage, parameterName=IsTorsoDamage}
  TextureSample() {texture=/Game/Characters/Player/Male/Medium/Bodies/M_MED_Doggus/Textures/T_Doggus_Torso_D}
  TextureSample() {texture=/FigmentEnvironment/FN_6_20/Characters/Fortress_Character/Textures/AllBlack}
 ]}
